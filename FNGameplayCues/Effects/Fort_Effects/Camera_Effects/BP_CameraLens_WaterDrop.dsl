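; Camera lens effect Blueprint that shows a screen-space waterflow post-process, named for water drops on the camera lens.
; On BeginPlay it creates a dynamic instance of a waterfall screen-waterflow material and applies it as a full-weight blendable on its PostProcess component.
; A timeline played from the start then drives the material's Time scalar parameter to animate the effect.
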(event EventBeginPlay
  (bind _returnvalue (Rendering|Material|CreateDynamicMaterialInstance "/FNGameplayCues/Athena/Artemis/Environments/BuildingActors/WaterFall/Materials/PPM_ScreenWaterflow_A_Inst.PPM_ScreenWaterflow_A_Inst" "None" "None" :!class /Script/Engine.KismetMaterialLibrary))
  (Variables|/Script/Engine.PostProcessComponent|SetSettings (Utilities|Struct|MakePostProcessSettings (Utilities|Struct|MakeWeightedBlendables (Utilities|Array|MakeArray (Utilities|Struct|MakeWeightedBlendable 1.0 _returnvalue)))) (Variables|Default|GetPostProcess))
  (bind (_direction _newtrack_0) (Timeline|Time PlayFromStart)
    (:Update
      (Rendering|Material|SetScalarParameterValue _returnvalue "Time" _newtrack_0 :!class /Script/Engine.MaterialInstanceDynamic))))

(fn ConstructionScript)
