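; Camera lens effect Blueprint, named for Slurp, that adapts its particle emitter to the rendering detail mode on BeginPlay.
; At detail mode 1 or higher it zeroes the emitter's VignettSpawnScale parameter.
; At low detail it sets that scale to full and plays a timeline that drives the emitter's Alpha parameter.
; Debug prints mark activation, low detail and destruction, while Tick and the construction script are empty.
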(event EventTick
  (params (DeltaSeconds :type Float%20%28single-precision%29)))

(event EventBeginPlay
  (bind _particlesystemcomponent (Variables|Emitter|GetParticleSystemComponent))
  (Development|PrintString "Active_Camera Lens Effect")
  (if (>= (Rendering|GetRenderingDetailMode) 1)
    (Effects|Components|ParticleSystem|SetFloatParameter _particlesystemcomponent "VignettSpawnScale" 0.0)
    (else
      (Development|PrintString "LOW Detail!")
      (Effects|Components|ParticleSystem|SetFloatParameter _particlesystemcomponent "VignettSpawnScale" 1.0)
      (bind (_direction _alpha) (Timeline|Timeline_0 PlayFromStart)
        (:Update
          (Effects|Components|ParticleSystem|SetFloatParameter _particlesystemcomponent "Alpha" _alpha))))))

(event EventDestroyed
  (Development|PrintString "Camera Lens - Destroyed"))

(fn ConstructionScript)
