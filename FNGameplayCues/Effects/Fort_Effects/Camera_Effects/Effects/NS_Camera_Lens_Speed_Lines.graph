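# Niagara system for a camera lens speed-lines effect, filed under the Fort camera effects for gameplay cues.
# It contains three emitters: LensEffectLoop, Streaks and Streaks001, whose names suggest a looping lens effect with two streak layers.
# SystemState is the only module listed, and no user parameters, data interfaces, dynamic inputs or renderers appear, so the emitters' internal behavior is not shown.
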NiagaraSystem /FNGameplayCues/Effects/Fort_Effects/Camera_Effects/Effects/NS_Camera_Lens_Speed_Lines
dependencies {usedDataInterfaces [], usedDynamicInputs [], usedModules [&SystemState]}
emitters {LensEffectLoop {rendererClasses [], renderers []}, Streaks {rendererClasses [], renderers []}, Streaks001 {rendererClasses [], renderers []}}
summary {
 systemName=NS_Camera_Lens_Speed_Lines
 emitters [{emitterName=LensEffectLoop, rendererClasses []}, {emitterName=Streaks, rendererClasses []}, {emitterName=Streaks001, rendererClasses []}]
 userVariables []
}
